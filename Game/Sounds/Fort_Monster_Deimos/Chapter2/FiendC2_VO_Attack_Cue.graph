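# Sound cue for the Chapter 2 Deimos Fiend monster's attack vocalization.
# A Random node picks one of nine equally weighted attack wave variants and feeds it through a Modulator at the root.
# It plays in the tactical creature vocalization class with close vocal attenuation and high priority.
# Virtualization is disabled, and global and local Deimos vocal concurrency sets limit overlapping voices.
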SoundCue /Game/Sounds/Fort_Monster_Deimos/Chapter2/FiendC2_VO_Attack_Cue
root=Modulator_0
nodes [
 {id=Modulator_0, children [Random_0]}
 {
  id=Random_0
  children [WavePlayer_0, WavePlayer_1, WavePlayer_2, WavePlayer_3, WavePlayer_4, WavePlayer_5, WavePlayer_6, WavePlayer_14, WavePlayer_15]
  properties {bShouldExcludeFromBranchCulling=true, weights [1,1,1,1,1,1,1,1,1]}
 }
 {id=WavePlayer_0, properties {soundWaveAssetPtr=&/Fiend_Attack_New_01}}
 {id=WavePlayer_1, properties {soundWaveAssetPtr=&/Fiend_Attack_New_02}}
 {id=WavePlayer_14, properties {soundWaveAssetPtr=&/Fiend_Attack_New_08}}
 {id=WavePlayer_15, properties {soundWaveAssetPtr=&/Fiend_Attack_New_09}}
 {id=WavePlayer_2, properties {soundWaveAssetPtr=&/Fiend_Attack_New_03}}
 {id=WavePlayer_3, properties {soundWaveAssetPtr=&/Fiend_Attack_New_04}}
 {id=WavePlayer_4, properties {soundWaveAssetPtr=&/Fiend_Attack_New_05}}
 {id=WavePlayer_5, properties {soundWaveAssetPtr=&/Fiend_Attack_New_06}}
 {id=WavePlayer_6, properties {soundWaveAssetPtr=&/Fiend_Attack_New_07}}
]
properties {
 attenuationSettings=/Game/Sounds/Attenuation/Fort_Deimos_Vocal_Close_Att
 bBypassVolumeScaleForPriority=true
 maxDistance=13000
 priority=0.9
 soundClassObject=&Fort_Creatures_Vocalization_Tactical
 virtualizationMode=Disabled
 volumeMultiplier=2.394
 attenuationOverrides {distanceAlgorithm=Logarithmic, falloffDistance=3000, attenuationShapeExtents {x=2000}}
 concurrencyOverrides {resolutionRule=PreventNew}
 concurrencySet [/Game/Sounds/Fort_Monster_Deimos/Settings/Deimos_Vocal_Global_Conc, /Game/Sounds/Fort_Monster_Deimos/Settings/Deimos_Vocal_Local_Conc]
}
