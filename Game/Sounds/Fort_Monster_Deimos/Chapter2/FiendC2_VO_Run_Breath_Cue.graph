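# Sound cue for the Chapter 2 Fiend monster's running breath vocalization.
# A single random node picks among twelve equally weighted RunBreathNew wave variations.
# It plays through the creature vocalization sound class with close vocal attenuation, using global and local Deimos vocal concurrency sets to limit overlapping voices.
# Virtualization is disabled, and volume scaling is bypassed when computing priority.
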SoundCue /Game/Sounds/Fort_Monster_Deimos/Chapter2/FiendC2_VO_Run_Breath_Cue
root=Random_1
nodes [
 {
  id=Random_1
  children [
   WavePlayer_26
   WavePlayer_27
   WavePlayer_28
   WavePlayer_29
   WavePlayer_30
   WavePlayer_31
   WavePlayer_32
   WavePlayer_33
   WavePlayer_34
   WavePlayer_35
   WavePlayer_36
   WavePlayer_37
  ]
  properties {bShouldExcludeFromBranchCulling=true, weights [1,1,1,1,1,1,1,1,1,1,1,1]}
 }
 {id=WavePlayer_26, properties {soundWaveAssetPtr=&/Fiend_RunBreathNew_10}}
 {id=WavePlayer_27, properties {soundWaveAssetPtr=&/Fiend_RunBreathNew_11}}
 {id=WavePlayer_28, properties {soundWaveAssetPtr=&/Fiend_RunBreathNew_12}}
 {id=WavePlayer_29, properties {soundWaveAssetPtr=&/Fiend_RunBreathNew_01}}
 {id=WavePlayer_30, properties {soundWaveAssetPtr=&/Fiend_RunBreathNew_02}}
 {id=WavePlayer_31, properties {soundWaveAssetPtr=&/Fiend_RunBreathNew_03}}
 {id=WavePlayer_32, properties {soundWaveAssetPtr=&/Fiend_RunBreathNew_04}}
 {id=WavePlayer_33, properties {soundWaveAssetPtr=&/Fiend_RunBreathNew_05}}
 {id=WavePlayer_34, properties {soundWaveAssetPtr=&/Fiend_RunBreathNew_06}}
 {id=WavePlayer_35, properties {soundWaveAssetPtr=&/Fiend_RunBreathNew_07}}
 {id=WavePlayer_36, properties {soundWaveAssetPtr=&/Fiend_RunBreathNew_08}}
 {id=WavePlayer_37, properties {soundWaveAssetPtr=&/Fiend_RunBreathNew_09}}
]
properties {
 attenuationSettings=/Game/Sounds/Attenuation/Fort_Deimos_Vocal_Close_Att
 bBypassVolumeScaleForPriority=true
 maxDistance=13000
 priority=0.75
 soundClassObject=/Game/Sounds/FortSoundClassesAndModes/Fort_Creatures_Vocalization_Normal
 virtualizationMode=Disabled
 volumeMultiplier=1.995
 attenuationOverrides {distanceAlgorithm=NaturalSound, falloffDistance=4000, attenuationShapeExtents {x=1000}}
 concurrencyOverrides {maxCount=2, volumeScale=0.75, volumeScaleAttackTime=0.1}
 concurrencySet [/Game/Sounds/Fort_Monster_Deimos/Settings/Deimos_Vocal_Global_Conc, /Game/Sounds/Fort_Monster_Deimos/Settings/Deimos_Vocal_Local_Conc]
}
